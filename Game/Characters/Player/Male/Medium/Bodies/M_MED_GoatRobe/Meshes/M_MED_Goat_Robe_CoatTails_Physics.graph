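# Physics asset for the coat tails of the medium male player's Goat Robe, simulating robe and front cloth bones against leg and foot bodies.
# Front, mid and back capsule chains on each side, plus short back and front cloth pieces, hang from the right thigh or left thigh-twist body with limited swing and twist.
# Extra constraints cross-link neighboring lower segments, mostly with free rotation, except an angularly locked left front-to-mid link.
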PhysicsAsset /Game/Characters/Player/Male/Medium/Bodies/M_MED_GoatRobe/Meshes/M_MED_Goat_Robe_CoatTails_Physics
bodies {
 calf_l [{shapeType=Capsule, length=31.5502, radius=9, center=<-23.4649 0.0001 -0.0969>, rotation=<rot 89.2922 -0.0004 -0.0005>}]
 calf_r [{shapeType=Capsule, length=31.5502, radius=9, center=<23.4649 0.0001 -0.0969>, rotation=<rot 89.2922 -0.0004 -0.0005>}]
 dyn_cloth_front_01 [{shapeType=Capsule, length=4.75, radius=2.5, shapeName=dyn_cloth_front_01_capsule, center=<3.1294 0 0>, rotation=<rot 90 0 0>}]
 dyn_cloth_front_02 [{shapeType=Capsule, length=4.75, radius=2.5, shapeName=dyn_cloth_front_02_capsule, center=<3.1294 0 0>, rotation=<rot 90 0 0>}]
 dyn_cloth_front_03 [{shapeType=Capsule, length=2.75, radius=2.5, shapeName=dyn_cloth_front_03_capsule, center=<3.1685 0 0>, rotation=<rot 90 0 0>}]
 dyn_robe_l_back_01 [{shapeType=Capsule, length=20, radius=7, shapeName=dyn_robe_l_back_01_capsule, center=<5.6936 0.0001 5>, rotation=<rot 0 0 90>}]
 dyn_robe_l_back_02 [{shapeType=Capsule, length=20, radius=10, shapeName=dyn_robe_l_back_02_capsule, center=<7.9151 0.0001 5>, rotation=<rot 0 0 90>}]
 dyn_robe_l_back_03 [{shapeType=Capsule, length=20, radius=10, shapeName=dyn_robe_l_back_03_capsule, center=<7.7314 0.0001 5>, rotation=<rot 0 0 90>}]
 dyn_robe_l_back_parent [{shapeType=Capsule, length=18, radius=7, shapeName=dyn_robe_l_back_parent_capsule, center=<6.9832 0 5>, rotation=<rot 0 0 90>}]
 dyn_robe_l_front_01 [{shapeType=Capsule, length=20, radius=8, shapeName=dyn_robe_l_front_01_capsule, center=<7.3444 0 8>, rotation=<rot 0 0 90>}]
 dyn_robe_l_front_02 [{shapeType=Capsule, length=20, radius=10, shapeName=dyn_robe_l_front_02_capsule, center=<6.2084 0 8>, rotation=<rot 0 0 90>}]
 dyn_robe_l_front_03 [{shapeType=Capsule, length=20, radius=10, shapeName=dyn_robe_l_front_03_capsule, center=<7.3444 0 8>, rotation=<rot 0 0 90>}]
 dyn_robe_l_front_parent [{shapeType=Capsule, length=20, radius=8, shapeName=dyn_robe_l_front_parent_capsule, center=<7.3444 0 10>, rotation=<rot 0 0 90>}]
 dyn_robe_l_mid_01 [{shapeType=Capsule, length=18, radius=7, shapeName=dyn_robe_l_mid_01_capsule, center=<5.7614 2 7>, rotation=<rot 0 0 90>}]
 dyn_robe_l_mid_02 [{shapeType=Capsule, length=20, radius=10, shapeName=dyn_robe_l_mid_02_capsule, center=<5.6247 2 7>, rotation=<rot 0 0 90>}]
 dyn_robe_l_mid_03 [{shapeType=Capsule, length=20, radius=10, shapeName=dyn_robe_l_mid_03_capsule, center=<5.6073 2 7>, rotation=<rot 0 0 90>}]
 dyn_robe_l_mid_parent [{shapeType=Capsule, length=18, radius=7, shapeName=dyn_robe_l_mid_parent_capsule, center=<6.476 2 7>, rotation=<rot 0 0 90>}]
 dyn_robe_r_back_01 [{shapeType=Capsule, length=18, radius=7, shapeName=dyn_robe_r_back_01_capsule, center=<6.8421 0 -5>, rotation=<rot 0 0 90>}]
 dyn_robe_r_back_02 [{shapeType=Capsule, length=20, radius=8, shapeName=dyn_robe_r_back_02_capsule, center=<6.8685 0 -5>, rotation=<rot 0 0 90>}]
 dyn_robe_r_back_03 [{shapeType=Capsule, length=20, radius=8, shapeName=dyn_robe_r_back_03_capsule, center=<9.492 0 -5>, rotation=<rot 0 0 90>}]
 dyn_robe_r_back_parent [{shapeType=Capsule, length=18, radius=7, shapeName=dyn_robe_r_back_parent_capsule, center=<6.8685 0 -5>, rotation=<rot 0 0 90>}]
 dyn_robe_r_front_01 [{shapeType=Capsule, length=20, radius=7, shapeName=dyn_robe_r_front_01_capsule, center=<9.7689 10 0>, rotation=<rot 0 0 15>}]
 dyn_robe_r_front_02 [{shapeType=Capsule, length=20, radius=7, shapeName=dyn_robe_r_front_02_capsule, center=<8.6329 9 0>, rotation=<rot 0 0 15>}]
 dyn_robe_r_front_03 [{shapeType=Capsule, length=20, radius=7, shapeName=dyn_robe_r_front_03_capsule, center=<9.7689 9 0>, rotation=<rot 0 0 15>}]
 dyn_robe_r_front_parent [{shapeType=Capsule, length=20, radius=7, shapeName=dyn_robe_r_front_parent_capsule, center=<9.7689 8 0>, rotation=<rot 0 0 15>}]
 dyn_robe_r_mid_01 [{shapeType=Capsule, length=18, radius=7, shapeName=dyn_robe_r_mid_01_capsule, center=<6.523 0 -5>, rotation=<rot 0 0 80>}]
 dyn_robe_r_mid_02 [{shapeType=Capsule, length=20, radius=8, shapeName=dyn_robe_r_mid_02_capsule, center=<6.4368 0 -5>, rotation=<rot 0 0 80>}]
 dyn_robe_r_mid_03 [{shapeType=Capsule, length=20, radius=8, shapeName=dyn_robe_r_mid_03_capsule, center=<6.5633 0 -5>, rotation=<rot 0 0 80>}]
 dyn_robe_r_mid_parent [{shapeType=Capsule, length=20, radius=7, shapeName=dyn_robe_r_mid_parent_capsule, center=<6.2536 0 -9>, rotation=<rot 0 0 80>}]
 dyn_robe_short_l_back_01 [{shapeType=Capsule, length=7.25, radius=3, shapeName=dyn_robe_short_l_back_01_capsule, center=<3.7691 -2 0>}]
 dyn_robe_short_l_back_02 [{shapeType=Capsule, length=4.5, radius=2.25, shapeName=dyn_robe_short_l_back_02_capsule, center=<3.6645 0 0>, rotation=<rot 90 0 0>}]
 dyn_robe_short_l_back_03 [{shapeType=Capsule, length=8, radius=3, shapeName=dyn_robe_short_l_back_03_capsule, center=<3.7253 -2 0>, rotation=<rot 90 0 0>}]
 dyn_robe_short_l_back_parent [{shapeType=Box, extentX=6.01, extentY=2.01, extentZ=6.51, center=<2.6287 0.0002 0>}]
 dyn_robe_short_r_back [{shapeType=Capsule, length=10, radius=3, shapeName=dyn_robe_short_r_back_capsule, center=<4.7151 -3 0>}]
 dyn_robe_short_r_back_01 [{shapeType=Capsule, length=8, radius=3, shapeName=dyn_robe_short_r_back_01_capsule, center=<4.2794 -3 0>}]
 foot_l [{shapeType=Capsule, length=17.4338, radius=5.6524, center=<-5.9725 7.3323 0.0004>, rotation=<rot 0 0 90>}]
 foot_r [{shapeType=Capsule, length=18.2288, radius=4.8573, center=<5.8738 -7.7009 0.0003>, rotation=<rot 0 0 90>}]
 thigh_r [{shapeType=Capsule, length=20, radius=8.5, center=<25.9421 -0.6605 -0.2585>, rotation=<rot 85.2847 179.9999 179.9999>}]
 thigh_twist_01_l [{shapeType=Capsule, length=20, radius=9, center=<-6.7507 0.4727 -0.2938>, rotation=<rot 90 0 0>}]
}
constraints [
 {bone1Name=dyn_robe_l_front_parent, bone2Name=thigh_twist_01_l, swing1LimitDegrees=20, swing2LimitDegrees=20, twistLimitDegrees=1}
 {bone1Name=dyn_robe_l_front_01, bone2Name=dyn_robe_l_front_parent, swing1LimitDegrees=20, swing2LimitDegrees=25, twistLimitDegrees=1}
 {bone1Name=dyn_robe_l_front_02, bone2Name=dyn_robe_l_front_01, swing1LimitDegrees=30, swing2LimitDegrees=35, twistLimitDegrees=1}
 {bone1Name=dyn_robe_l_back_parent, bone2Name=thigh_twist_01_l, swing1LimitDegrees=35, swing2LimitDegrees=35, twistLimitDegrees=10}
 {bone1Name=dyn_robe_l_back_01, bone2Name=dyn_robe_l_back_parent, swing1LimitDegrees=15, swing2LimitDegrees=25, twistLimitDegrees=8}
 {bone1Name=dyn_robe_l_back_02, bone2Name=dyn_robe_l_back_01, swing1LimitDegrees=15, swing2LimitDegrees=25, twistLimitDegrees=8}
 {bone1Name=dyn_robe_short_l_back_parent, bone2Name=thigh_twist_01_l, swing1LimitDegrees=20, swing2LimitDegrees=20, twistLimitDegrees=6}
 {
  bone1Name=dyn_robe_short_l_back_01
  bone2Name=dyn_robe_short_l_back_parent
  swing1LimitDegrees=20
  swing2LimitDegrees=20
  twistLimitDegrees=6
 }
 {bone1Name=dyn_robe_short_l_back_03, bone2Name=dyn_robe_short_l_back_01, swing1LimitDegrees=20, swing2LimitDegrees=10, twistLimitDegrees=6}
 {
  bone1Name=dyn_robe_short_l_back_02
  bone2Name=dyn_robe_short_l_back_parent
  swing1LimitDegrees=25
  swing2LimitDegrees=25
  twistLimitDegrees=6
 }
 {bone1Name=dyn_robe_r_front_parent, bone2Name=thigh_r, swing1LimitDegrees=20, swing2LimitDegrees=20, twistLimitDegrees=1}
 {bone1Name=dyn_robe_r_front_02, bone2Name=dyn_robe_r_front_01, swing1LimitDegrees=20, swing2LimitDegrees=35, twistLimitDegrees=1}
 {bone1Name=dyn_robe_r_back_parent, bone2Name=thigh_r, swing1LimitDegrees=35, swing2LimitDegrees=35, twistLimitDegrees=10}
 {bone1Name=dyn_robe_r_back_01, bone2Name=dyn_robe_r_back_parent, swing1LimitDegrees=15, swing2LimitDegrees=25, twistLimitDegrees=8}
 {bone1Name=dyn_robe_short_r_back, bone2Name=thigh_r, swing1LimitDegrees=20, swing2LimitDegrees=20, twistLimitDegrees=8}
 {bone1Name=dyn_robe_r_back_02, bone2Name=dyn_robe_r_back_01, swing1LimitDegrees=15, swing2LimitDegrees=25, twistLimitDegrees=5}
 {bone1Name=dyn_robe_short_r_back_01, bone2Name=dyn_robe_short_r_back, swing1LimitDegrees=20, swing2LimitDegrees=20, twistLimitDegrees=8}
 {bone1Name=dyn_cloth_front_02, bone2Name=dyn_cloth_front_01, swing1LimitDegrees=45, swing2LimitDegrees=45, twistLimitDegrees=45}
 {bone1Name=dyn_robe_l_mid_parent, bone2Name=thigh_twist_01_l, swing1LimitDegrees=35, swing2LimitDegrees=35, twistLimitDegrees=10}
 {bone1Name=dyn_robe_l_mid_01, bone2Name=dyn_robe_l_mid_parent, swing1LimitDegrees=15, swing2LimitDegrees=25, twistLimitDegrees=8}
 {bone1Name=dyn_robe_l_mid_02, bone2Name=dyn_robe_l_mid_01, swing1LimitDegrees=15, swing2LimitDegrees=25, twistLimitDegrees=8}
 {bone1Name=dyn_robe_r_mid_parent, bone2Name=thigh_r, swing1LimitDegrees=35, swing2LimitDegrees=35, twistLimitDegrees=10}
 {bone1Name=dyn_robe_r_mid_01, bone2Name=dyn_robe_r_mid_parent, swing1LimitDegrees=15, swing2LimitDegrees=25, twistLimitDegrees=8}
 {bone1Name=dyn_robe_r_mid_02, bone2Name=dyn_robe_r_mid_01, swing1LimitDegrees=15, swing2LimitDegrees=25, twistLimitDegrees=5}
 {bone1Name=dyn_robe_r_mid_03, bone2Name=dyn_robe_r_mid_02, swing1LimitDegrees=15, swing2LimitDegrees=20, twistLimitDegrees=5}
 {bone1Name=dyn_robe_r_back_03, bone2Name=dyn_robe_r_back_02, swing1LimitDegrees=55, swing2LimitDegrees=55, twistLimitDegrees=5}
 {bone1Name=dyn_robe_r_front_03, bone2Name=dyn_robe_r_front_02, swing1LimitDegrees=20, swing2LimitDegrees=55, twistLimitDegrees=1}
 {bone1Name=dyn_robe_r_front_01, bone2Name=dyn_robe_r_front_parent, swing1LimitDegrees=20, swing2LimitDegrees=25, twistLimitDegrees=1}
 {bone1Name=dyn_robe_l_front_03, bone2Name=dyn_robe_l_front_02, swing1LimitDegrees=50, swing2LimitDegrees=55, twistLimitDegrees=1}
 {bone1Name=dyn_robe_l_back_03, bone2Name=dyn_robe_l_back_02, swing1LimitDegrees=30, swing2LimitDegrees=30, twistLimitDegrees=5}
 {bone1Name=dyn_robe_l_mid_03, bone2Name=dyn_robe_l_mid_02, swing1LimitDegrees=15, swing2LimitDegrees=20, twistLimitDegrees=5}
 {
  bone1Name=dyn_robe_r_mid_02
  bone2Name=dyn_robe_r_front_03
  swing1LimitDegrees=45
  swing1Motion=Free
  swing2LimitDegrees=45
  swing2Motion=Free
  twistLimitDegrees=45
  twistMotion=Free
 }
 {
  bone1Name=dyn_robe_r_mid_02
  bone2Name=dyn_robe_r_mid_03
  swing1LimitDegrees=45
  swing1Motion=Free
  swing2LimitDegrees=45
  swing2Motion=Free
  twistLimitDegrees=45
  twistMotion=Free
 }
 {
  bone1Name=dyn_robe_r_mid_02
  bone2Name=dyn_robe_r_front_02
  swing1LimitDegrees=45
  swing1Motion=Free
  swing2LimitDegrees=45
  swing2Motion=Free
  twistLimitDegrees=45
  twistMotion=Free
 }
 {
  bone1Name=dyn_robe_r_mid_03
  bone2Name=dyn_robe_r_back_03
  swing1LimitDegrees=45
  swing1Motion=Free
  swing2LimitDegrees=45
  swing2Motion=Free
  twistLimitDegrees=45
  twistMotion=Free
 }
 {
  bone1Name=dyn_robe_r_mid_03
  bone2Name=dyn_robe_r_front_03
  swing1LimitDegrees=45
  swing1Motion=Free
  swing2LimitDegrees=45
  swing2Motion=Free
  twistLimitDegrees=45
  twistMotion=Free
 }
 {bone1Name=dyn_cloth_front_01, bone2Name=thigh_r, swing1LimitDegrees=15, swing2LimitDegrees=15, twistLimitDegrees=5}
 {bone1Name=dyn_cloth_front_03, bone2Name=thigh_twist_01_l, swing1LimitDegrees=20, swing2LimitDegrees=20, twistLimitDegrees=5}
 {
  bone1Name=dyn_robe_l_back_02
  bone2Name=dyn_robe_l_mid_02
  swing1LimitDegrees=45
  swing1Motion=Free
  swing2LimitDegrees=45
  swing2Motion=Free
  twistLimitDegrees=45
  twistMotion=Free
 }
 {
  bone1Name=dyn_robe_l_front_02
  bone2Name=dyn_robe_l_mid_02
  swing1LimitDegrees=45
  swing1Motion=Locked
  swing2LimitDegrees=45
  swing2Motion=Locked
  twistLimitDegrees=45
  twistMotion=Locked
 }
 {
  bone1Name=dyn_robe_l_mid_03
  bone2Name=dyn_robe_l_back_03
  swing1LimitDegrees=45
  swing1Motion=Free
  swing2LimitDegrees=45
  swing2Motion=Free
  twistLimitDegrees=45
  twistMotion=Free
 }
 {
  bone1Name=dyn_robe_l_mid_03
  bone2Name=dyn_robe_l_front_03
  swing1LimitDegrees=45
  swing1Motion=Free
  swing2LimitDegrees=45
  swing2Motion=Free
  twistLimitDegrees=45
  twistMotion=Free
 }
 {
  bone1Name=dyn_robe_l_mid_03
  bone2Name=dyn_robe_l_mid_02
  swing1LimitDegrees=45
  swing1Motion=Free
  swing2LimitDegrees=45
  swing2Motion=Free
  twistLimitDegrees=45
  twistMotion=Free
 }
]
