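# Physics asset for the fur on the medium male player's goat robe body, defining bodies and joints for its secondary motion.
# Six box bodies on the dyn_fur bones are each constrained directly to a dyn_fur_parent capsule, forming a star rather than a chain.
# Tight swing and twist limits of a few degrees keep the fur's movement subtle.
# A left clavicle capsule has no constraint, apparently serving as a collider.
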PhysicsAsset /Game/Characters/Player/Male/Medium/Bodies/M_MED_GoatRobe/Meshes/M_MED_Goat_Robe_Fur_Physics
bodies {
 clavicle_l [{shapeType=Capsule, length=5.1374, radius=4.1072}]
 dyn_fur_01 [{shapeType=Box, extentX=3.8215, extentY=3.2972, extentZ=1.9548, center=<1.2774 0 0>}]
 dyn_fur_02 [{shapeType=Box, extentX=3.1463, extentY=2.2794, extentZ=1.9087, center=<0.851 0 0>}]
 dyn_fur_03 [{shapeType=Box, extentX=4.6439, extentY=2.6979, extentZ=2.3903, center=<0.9457 0 0>}]
 dyn_fur_04 [{shapeType=Box, extentX=5.0183, extentY=3.0739, extentZ=2.6066, center=<1.4891 0 0>}]
 dyn_fur_05 [{shapeType=Box, extentX=3.0677, extentY=2.4378, extentZ=1.6931, center=<0.7831 0 0>}]
 dyn_fur_06 [{shapeType=Box, extentX=2.9771, extentY=1.6555, extentZ=2, center=<0.7044 0 0>}]
 dyn_fur_parent [{shapeType=Capsule, length=6.2463, radius=6.1287, center=<1.0582 -0.9841 -4.7999>, rotation=<rot 52.6026 14.2508 49.0527>}]
}
constraints [
 {bone1Name=dyn_fur_01, bone2Name=dyn_fur_parent, swing1LimitDegrees=4, swing2LimitDegrees=4, twistLimitDegrees=2}
 {bone1Name=dyn_fur_02, bone2Name=dyn_fur_parent, swing1LimitDegrees=4, swing2LimitDegrees=4, twistLimitDegrees=2}
 {bone1Name=dyn_fur_03, bone2Name=dyn_fur_parent, swing1LimitDegrees=4, swing2LimitDegrees=4, twistLimitDegrees=2}
 {bone1Name=dyn_fur_04, bone2Name=dyn_fur_parent, swing1LimitDegrees=4, swing2LimitDegrees=4, twistLimitDegrees=2}
 {bone1Name=dyn_fur_05, bone2Name=dyn_fur_parent, swing1LimitDegrees=3, swing2LimitDegrees=3, twistLimitDegrees=2}
 {bone1Name=dyn_fur_06, bone2Name=dyn_fur_parent, swing1LimitDegrees=4, swing2LimitDegrees=4, twistLimitDegrees=2}
]
