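# Physics asset for the back scarf on the medium male player's GoatRobe body, set up for secondary motion of the cloth piece.
# A short bone chain runs from a box body at the scarf base through capsule bodies on the parent and 01 bones.
# Two constraints link the chain with angular limits, keeping twist much tighter than swing.
# Three large capsules on spine_05, unconstrained to the chain, approximate the upper torso, apparently as a collider for the scarf.
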PhysicsAsset /Game/Characters/Player/Male/Medium/Bodies/M_MED_GoatRobe/Meshes/M_MED_Goat_Robe_Scarf_Back_Physics
bodies {
 dyn_scarf_back_01 [{shapeType=Capsule, length=8, radius=1.3, shapeName=dyn_scarf_back_01_capsule, center=<2.8567 -1.5 0>}]
 dyn_scarf_back_base [{shapeType=Box, extentX=2.25, extentY=1.5, extentZ=2}]
 dyn_scarf_back_parent [{shapeType=Capsule, length=23.75, radius=2.25, shapeName=dyn_scarf_back_parent_capsule, center=<3.1204 -2 0>}]
 spine_05 [
  {shapeType=Capsule, length=18.25, radius=11.75}
  {shapeType=Capsule, length=21.625, radius=12.5, center=<-11.4663 0.2189 0>, rotation=<rot 90 180 176.2303>}
  {shapeType=Capsule, length=23.5, radius=7.5, shapeName=spine_05_capsule, center=<-12.2007 -5.2708 0>}
 ]
}
constraints [
 {bone1Name=dyn_scarf_back_parent, bone2Name=dyn_scarf_back_base, swing1LimitDegrees=25, swing2LimitDegrees=13, twistLimitDegrees=3}
 {bone1Name=dyn_scarf_back_01, bone2Name=dyn_scarf_back_parent, swing1LimitDegrees=20, swing2LimitDegrees=13, twistLimitDegrees=3}
]
