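# Physics asset for the scarf on the medium male player's Goat Robe body, defining collision and joint limits for its dynamic scarf bones.
# The back chain runs base, parent, 01 using boxes, while the front runs base, parent, 01, 02 using capsules below a box base.
# Constraints allow moderate swing and very little twist.
# Two large spine_05 capsules give the upper torso a collision volume, with no constraint linking it to either chain.
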PhysicsAsset /Game/Characters/Player/Male/Medium/Bodies/M_MED_GoatRobe/Meshes/M_MED_Goat_Robe_Scarf_Physics
bodies {
 dyn_scarf_back_01 [{shapeType=Box, extentX=6.4311, extentY=1.3208, extentZ=12.5558, center=<2.8567 0 0>}]
 dyn_scarf_back_base [{shapeType=Box, extentX=2.25, extentY=1.5, extentZ=2}]
 dyn_scarf_back_parent [{shapeType=Box, extentX=6.6883, extentY=1.3971, extentZ=12.7642, center=<3.1204 0 0>}]
 dyn_scarf_front_01 [{shapeType=Capsule, length=8.5, radius=2.5, shapeName=dyn_scarf_front_01_capsule, center=<3.4527 2 0>}]
 dyn_scarf_front_02 [{shapeType=Capsule, length=8.5, radius=2.5, shapeName=dyn_scarf_front_02_capsule, center=<2.4312 2 0>}]
 dyn_scarf_front_base [{shapeType=Box, extentX=1.5, extentY=1.5, extentZ=1.5}]
 dyn_scarf_front_parent [{shapeType=Capsule, length=8.5, radius=2.5, shapeName=dyn_scarf_front_parent_capsule, center=<2.3956 2 0>}]
 spine_05 [
  {shapeType=Capsule, length=18.25, radius=11.75}
  {shapeType=Capsule, length=21.625, radius=12.5, center=<-11.4663 0.2189 0>, rotation=<rot 90 0 -3.7697>}
 ]
}
constraints [
 {bone1Name=dyn_scarf_back_parent, bone2Name=dyn_scarf_back_base, swing1LimitDegrees=25, swing2LimitDegrees=13, twistLimitDegrees=3}
 {bone1Name=dyn_scarf_back_01, bone2Name=dyn_scarf_back_parent, swing1LimitDegrees=20, swing2LimitDegrees=13, twistLimitDegrees=3}
 {bone1Name=dyn_scarf_front_parent, bone2Name=dyn_scarf_front_base, swing1LimitDegrees=20, swing2LimitDegrees=20, twistLimitDegrees=3}
 {bone1Name=dyn_scarf_front_01, bone2Name=dyn_scarf_front_parent, swing1LimitDegrees=20, swing2LimitDegrees=20, twistLimitDegrees=3}
 {bone1Name=dyn_scarf_front_02, bone2Name=dyn_scarf_front_01, swing1LimitDegrees=20, swing2LimitDegrees=20, twistLimitDegrees=3}
]
