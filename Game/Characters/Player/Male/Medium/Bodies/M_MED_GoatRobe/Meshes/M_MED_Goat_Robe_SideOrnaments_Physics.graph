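# Physics asset for the side ornaments on the male medium player's goat robe: a dangling rope and a book.
# The rope is a chain of capsules hung from a small box base with uniform tight swing and twist limits.
# The book is a single capsule on its own box base, limited to swing mostly along one axis.
# Unconstrained capsules on the left thigh and calf provide leg shapes alongside the ornaments, likely for collision.
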PhysicsAsset /Game/Characters/Player/Male/Medium/Bodies/M_MED_GoatRobe/Meshes/M_MED_Goat_Robe_SideOrnaments_Physics
bodies {
 calf_l [{shapeType=Capsule, length=32.125, radius=11, center=<-15.1994 -3.4428 -0.3832>, rotation=<rot 90 0 0>}]
 dyn_book_base [{shapeType=Box, extentX=1.01, extentY=1.01, extentZ=1.01, rotation=<rot 0 -90 0>}]
 dyn_book_parent [{shapeType=Capsule, length=10.75, radius=4.25, shapeName=dyn_book_parent_capsule, center=<5.79 0 0>, rotation=<rot -0.0001 0.0003 84.3748>}]
 dyn_rope_01 [{shapeType=Capsule, length=5.7024, radius=2.5226, center=<4.9939 0 0>, rotation=<rot 90 0 0>}]
 dyn_rope_02 [{shapeType=Capsule, length=8.1596, radius=2.4363, center=<5.9951 0 0>, rotation=<rot 90 0 0>}]
 dyn_rope_03 [{shapeType=Capsule, length=6.3334, radius=5.3094, center=<7.6803 0 0>, rotation=<rot 90 0 0>}]
 dyn_rope_base [{shapeType=Box, extentX=1.5, extentY=1.5, extentZ=1.5}]
 dyn_rope_parent [{shapeType=Capsule, length=5.8348, radius=3.2641, center=<4.6455 0 0>, rotation=<rot 90 0 0>}]
 thigh_twist_01_l [{shapeType=Capsule, length=30.875, radius=10.125, center=<0.0834 -0.0002 -0.5203>, rotation=<rot -80.8917 0 0>}]
}
constraints [
 {bone1Name=dyn_rope_parent, bone2Name=dyn_rope_base, swing1LimitDegrees=10, swing2LimitDegrees=10, twistLimitDegrees=20}
 {bone1Name=dyn_rope_01, bone2Name=dyn_rope_parent, swing1LimitDegrees=10, swing2LimitDegrees=10, twistLimitDegrees=20}
 {bone1Name=dyn_rope_02, bone2Name=dyn_rope_01, swing1LimitDegrees=10, swing2LimitDegrees=10, twistLimitDegrees=20}
 {bone1Name=dyn_rope_03, bone2Name=dyn_rope_02, swing1LimitDegrees=10, swing2LimitDegrees=10, twistLimitDegrees=20}
 {bone1Name=dyn_book_parent, bone2Name=dyn_book_base, swing1LimitDegrees=5, swing2LimitDegrees=45, twistLimitDegrees=2}
]
