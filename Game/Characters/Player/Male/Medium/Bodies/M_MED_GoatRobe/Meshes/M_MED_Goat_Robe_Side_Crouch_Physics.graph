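# Physics asset for the male medium player's Goat Robe crouch side mesh, covering a dangling rope and a book accessory.
# The rope is a chain of capsules hung from a small box base, with wide swing limits and mostly tight twist limits along its middle links.
# The book is a box hinged to its base with narrow limits, swinging mainly on one axis.
# Left thigh and calf capsules have no constraints, likely serving as leg colliders.
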PhysicsAsset /Game/Characters/Player/Male/Medium/Bodies/M_MED_GoatRobe/Meshes/M_MED_Goat_Robe_Side_Crouch_Physics
bodies {
 calf_l [{shapeType=Capsule, length=35.1131, radius=8.0119, center=<-15.1994 -3.4428 -0.3832>, rotation=<rot 90 0 0>}]
 dyn_book_base [{shapeType=Box, extentX=1.01, extentY=1.01, extentZ=1.01, rotation=<rot 0 -90 0>}]
 dyn_book_parent [{shapeType=Box, extentX=13.871, extentY=19.8623, extentZ=4.8777, center=<7.4745 -1.3285 -1.4934>, rotation=<rot -9.7657 0 -8.7183>}]
 dyn_rope_01 [{shapeType=Capsule, length=6.4524, radius=1.7726, center=<4.9939 0 0>, rotation=<rot 90 0 0>}]
 dyn_rope_02 [{shapeType=Capsule, length=9.6596, radius=1.9363, center=<5.9951 0 0>, rotation=<rot 90 0 0>}]
 dyn_rope_03 [{shapeType=Capsule, length=6.5834, radius=5.0594, center=<7.2384 0 0>, rotation=<rot 90 0 0>}]
 dyn_rope_base [{shapeType=Box, extentX=1.5, extentY=1.5, extentZ=1.5}]
 dyn_rope_parent [{shapeType=Capsule, length=7.3348, radius=2.0141, center=<5.0793 0 0>, rotation=<rot 90 0 0>}]
 thigh_twist_01_l [{shapeType=Capsule, length=32.5486, radius=8.4513, center=<-0.0373 -0.0211 -0.5512>, rotation=<rot -82.4716 -2.0476 11.5942>}]
}
constraints [
 {bone1Name=dyn_rope_parent, bone2Name=dyn_rope_base, swing1LimitDegrees=80, swing2LimitDegrees=80, twistLimitDegrees=80}
 {bone1Name=dyn_rope_01, bone2Name=dyn_rope_parent, swing1LimitDegrees=90, swing2LimitDegrees=90, twistLimitDegrees=10}
 {bone1Name=dyn_rope_02, bone2Name=dyn_rope_01, swing1LimitDegrees=90, swing2LimitDegrees=90, twistLimitDegrees=10}
 {bone1Name=dyn_rope_03, bone2Name=dyn_rope_02, swing1LimitDegrees=90, swing2LimitDegrees=90, twistLimitDegrees=60}
 {bone1Name=dyn_book_parent, bone2Name=dyn_book_base, swing1LimitDegrees=5, swing2LimitDegrees=45, twistLimitDegrees=3}
]
