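# Post-process-style AnimBP that adds secondary motion to the Goat Robe player body's cloth, ornaments, scarf, fur and book.
# It copies the parent mesh pose, adjusts rope and back scarf, then caches it for crouch, jump/run and emote branches combining RigidBody sims with bone offsets and constraints.
# Blended branches feed string constraints, AnimDynamics chains, arm-swing fur offsets and a crouch LookAt.
# The event graph updates pawn-based state, including gravity override and frontend alpha.
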AnimBlueprint /Game/Characters/Player/Male/Medium/Bodies/M_MED_GoatRobe/Meshes/M_MED_Goat_Robe_AnimBp
target_skeleton=&/M_MED_Goat_Robe_Skeleton
graphs [
 AnimGraph {nodes [
   n1 = Anim Dynamics - Bone: dyn_sleeve_01(
    ComponentPose=Reroute Node(InputPin:Component Space Pose Link Structure) -> (OutputPin:Component Space Pose Link Structure)
    Alpha=1
    ExternalForce="0,0,0"
   ) {
    angularDampingOverride=0.999
    angularSpringConstant=5000
    bGravityOverrideInSimSpace=true
    bOverrideAngularDamping=true
    gravityScale=0.25
    alphaBoolBlend {blendOption=Linear}
    alphaScaleBiasClamp {scale=0.2}
    boundBone {boneName=dyn_sleeve_01}
    componentAppliedLinearAccClamp=<600 600 600>
    componentLinearAccScale=<0.25 0.25 0.25>
    componentLinearVelScale=<0.25 0.25 0.25>
    physicsBodyDefinitions [{
      sphereCollisionRadius=0
      boundBone {boneName=dyn_sleeve_01}
      boxExtents=<8 8 3>
      constraintSetup {angularLimitsMax=<0 20 0>, angularLimitsMin=<0 -20 0>, angularTarget=<1 0 0>}
      localJointOffset=<2 0 0>
     }]
    relativeSpaceBone {boneName=root}
   }
   n2 = Anim Dynamics - Bone: dyn_sleeve_02(ComponentPose=@n1, Alpha=1, ExternalForce="0,0,0") {
    angularDampingOverride=0.999
    angularSpringConstant=5000
    bGravityOverrideInSimSpace=true
    bOverrideAngularDamping=true
    gravityScale=0.25
    alphaBoolBlend {blendOption=Linear}
    alphaScaleBiasClamp {scale=0.2}
    boundBone {boneName=dyn_sleeve_02}
    componentAppliedLinearAccClamp=<600 600 600>
    componentLinearAccScale=<0.25 0.25 0.25>
    componentLinearVelScale=<0.25 0.25 0.25>
    physicsBodyDefinitions [{
      sphereCollisionRadius=0
      boundBone {boneName=dyn_sleeve_02}
      boxExtents=<8 8 3>
      constraintSetup {angularLimitsMax=<0 20 0>, angularLimitsMin=<0 -20 0>, angularTarget=<1 0 0>}
      localJointOffset=<2 0 0>
     }]
    relativeSpaceBone {boneName=root}
   }
   Anim Dynamics - Bone: dyn_sleeve_03(ComponentPose=@n2, Alpha=1, ExternalForce="0,0,0") {
    angularDampingOverride=0.999
    angularSpringConstant=5000
    bGravityOverrideInSimSpace=true
    bOverrideAngularDamping=true
    gravityScale=0.25
    alphaBoolBlend {blendOption=Linear}
    alphaScaleBiasClamp {scale=0.2}
    boundBone {boneName=dyn_sleeve_03}
    componentAppliedLinearAccClamp=<600 600 600>
    componentLinearAccScale=<0.25 0.25 0.25>
    componentLinearVelScale=<0.25 0.25 0.25>
    physicsBodyDefinitions [{
      sphereCollisionRadius=0
      boundBone {boneName=dyn_sleeve_03}
      boxExtents=<8 8 3>
      constraintSetup {angularLimitsMax=<0 20 0>, angularLimitsMin=<0 -20 0>, angularTarget=<1 0 0>}
      localJointOffset=<2 0 0>
     }]
    relativeSpaceBone {boneName=root}
   }
   RigidBody() {
    bClampLinearTranslationLimitToRefPose=true
    bOverrideWorldGravity=true
    overridePhysicsAsset=&/M_MED_Goat_Robe_Fur_Physics
    simulationSpace=BaseBoneSpace
    alphaBoolBlend {blendOption=Linear}
    baseBoneRef {boneName=spine_01}
    componentAppliedLinearAccClamp=<4000 4000 4000>
    componentLinearAccScale=<0.25 0.25 0.25>
    componentLinearVelScale=<0.25 0.25 0.25>
    overrideWorldGravity {z=-700}
   }
   RigidBody() {
    bClampLinearTranslationLimitToRefPose=true
    bOverrideWorldGravity=true
    overridePhysicsAsset=&/M_MED_Goat_Robe_Scarf_Back_Physics
    simulationSpace=BaseBoneSpace
    alphaBoolBlend {blendOption=Linear}
    baseBoneRef {boneName=pelvis}
    componentAppliedLinearAccClamp=<6000 6000 6000>
    componentLinearAccScale=<0.25 0.25 0.5>
    componentLinearVelScale=<0.25 0.25 0.5>
    overrideWorldGravity {z=-825}
   }
   RigidBody() {
    bClampLinearTranslationLimitToRefPose=true
    bOverrideWorldGravity=true
    overridePhysicsAsset=&/M_MED_Goat_Robe_Scarf_Physics
    simulationSpace=BaseBoneSpace
    alphaBoolBlend {blendOption=Linear}
    baseBoneRef {boneName=spine_05}
    componentAppliedLinearAccClamp=<6000 6000 6000>
    componentLinearAccScale=<0.1 0.1 0.2>
    componentLinearVelScale=<0.1 0.1 0.2>
    overrideWorldGravity {z=-825}
    simSpaceSettings {
     maxAngularAcceleration=500
     maxAngularVelocity=3
     maxLinearAcceleration=400
     maxLinearVelocity=250
     velocityScaleZ=0
     worldAlpha=1
     externalLinearDragV=<0.1 0.1>
    }
   }
   n3 = Get armSwingDelta_l() -> (armSwingDelta_l:double)
   n4 = Get CrouchRigid() -> (CrouchRigid:double)
   n5 = RigidBody(ComponentPose=Local To Component(LocalPose=Use cached pose 'GoatRobe_Cache'()), Alpha=@n4) {
    bClampLinearTranslationLimitToRefPose=true
    bOverrideWorldGravity=true
    overridePhysicsAsset=&/M_MED_Goat_Robe_Side_Crouch_Physics
    simulationSpace=BaseBoneSpace
    alphaBoolBlend {blendOption=Linear}
    alphaScaleBiasClamp {interpSpeedDecreasing=0.1, interpSpeedIncreasing=0.1}
    baseBoneRef {boneName=pelvis}
    componentAppliedLinearAccClamp=<3000 3000 3000>
    componentLinearAccScale=<0.5 0.4 0.4>
    componentLinearVelScale=<0.5 0.4 0.4>
    overrideWorldGravity {z=-800}
    simSpaceSettings {
     maxAngularAcceleration=500
     maxAngularVelocity=3
     maxLinearAcceleration=400
     maxLinearVelocity=250
     velocityScaleZ=0
     worldAlpha=1
     externalLinearDragV=<0.1 0.1>
    }
   }
   n6 = RigidBody(ComponentPose=@n5, Alpha=@n4) {
    bClampLinearTranslationLimitToRefPose=true
    bOverrideWorldGravity=true
    overridePhysicsAsset=&/M_MED_Goat_Robe_CoatTails_Crouch_Physics
    simulationSpace=BaseBoneSpace
    alphaBoolBlend {blendOption=Linear}
    alphaScaleBiasClamp {clampMax=0.1, clampMin=1, interpSpeedIncreasing=1, outRange {max=0, min=1}}
    baseBoneRef {boneName=pelvis}
    componentAppliedLinearAccClamp=<3000 3000 3000>
    componentLinearAccScale=<0.5 0.4 0.4>
    componentLinearVelScale=<0.5 0.4 0.4>
    overrideWorldGravity {z=-800}
    simSpaceSettings {
     maxAngularAcceleration=500
     maxAngularVelocity=3
     maxLinearAcceleration=400
     maxLinearVelocity=250
     velocityScaleZ=0
     worldAlpha=1
     externalLinearDragV=<0.1 0.1>
    }
   }
   n7 = "Transform (Modify) Bone - Bone: dyn_robe_l_front_base"(ComponentPose=@n6, Alpha=1, Translation="1,0,3", Rotation="0,0,0", Scale="1,1,1") {
    rotationMode=BMM_Additive
    rotationSpace=BCS_BoneSpace
    translationMode=BMM_Additive
    translationSpace=BCS_BoneSpace
    alphaBoolBlend {blendOption=Linear}
    alphaScaleBiasClamp {bInterpResult=true, interpSpeedDecreasing=5, interpSpeedIncreasing=8}
    boneToModify {boneName=dyn_robe_l_front_base}
    translation {x=1, z=3}
   }
   n8 = "Transform (Modify) Bone - Bone: dyn_robe_l_mid_base"(ComponentPose=@n7, Alpha=1, Translation="1,0,2", Rotation="0,0,0", Scale="1,1,1") {
    rotationMode=BMM_Additive
    rotationSpace=BCS_BoneSpace
    translationMode=BMM_Additive
    translationSpace=BCS_BoneSpace
    alphaBoolBlend {blendOption=Linear}
    alphaScaleBiasClamp {bInterpResult=true, interpSpeedDecreasing=5, interpSpeedIncreasing=8}
    boneToModify {boneName=dyn_robe_l_mid_base}
    translation {x=1, z=2}
   }
   n9 = Look At - Bone: dyn_robe_short_l_back_01(ComponentPose=@n8, Alpha=1) {
    alphaBoolBlend {blendOption=Linear}
    boneToModify {boneName=dyn_robe_short_l_back_01}
    lookAtLocation=<0 100>
    lookAtTarget {boneReference {boneName=calf_l}}
   }
   n10 = Get Jump_Alpha() -> (Jump_Alpha:double)
   n11 = "Transform (Modify) Bone - Bone: dyn_robe_r_back_base"(
    ComponentPose=Local To Component(LocalPose=Use cached pose 'GoatRobe_Cache'())
    Alpha=@n10
    Translation="0,0,0"
    Rotation="0,-60,0"
    Scale="1,1,1"
   ) {
    rotationMode=BMM_Additive
    rotationSpace=BCS_BoneSpace
    alphaBoolBlend {blendInTime=0.15, blendOption=Linear, blendOutTime=0.15}
    alphaScaleBiasClamp {bInterpResult=true, interpSpeedDecreasing=1, interpSpeedIncreasing=1}
    boneToModify {boneName=dyn_robe_r_back_base}
    rotation {yaw=-60}
   }
   n12 = "Transform (Modify) Bone - Bone: dyn_robe_r_back_base"(ComponentPose=@n11, Alpha=@n10, Translation="0,0,0", Rotation="0,-60,0", Scale="1,1,1") {
    rotationMode=BMM_Additive
    rotationSpace=BCS_BoneSpace
    alphaBoolBlend {blendInTime=0.15, blendOption=Linear, blendOutTime=0.15}
    alphaScaleBiasClamp {bInterpResult=true, interpSpeedDecreasing=1, interpSpeedIncreasing=1}
    boneToModify {boneName=dyn_robe_r_back_base}
    rotation {yaw=-60}
   }
   n13 = "Transform (Modify) Bone - Bone: dyn_robe_l_front_01"!(ComponentPose=@n12, Alpha=@n10, Translation="0,0,5", Rotation="15.000,0.000,-0.000") -> (Pose) {
    rotationMode=BMM_Replace
    rotationSpace=BCS_BoneSpace
    translationMode=BMM_Additive
    translationSpace=BCS_BoneSpace
    alphaBoolBlend {blendInTime=0.15, blendOption=Linear, blendOutTime=0.15}
    alphaScaleBiasClamp {bInterpResult=true, interpSpeedDecreasing=1, interpSpeedIncreasing=4}
    boneToModify {boneName=dyn_robe_l_front_01}
    rotation=<rot 15.0001 0.0001 0>
    translation {z=5}
   }
   n14 = "Transform (Modify) Bone - Bone: dyn_robe_r_front_01"!(ComponentPose=@n13, Alpha=@n10, Translation="0,0,5", Rotation="0.000,20.000,0.000") -> (Pose) {
    rotationMode=BMM_Replace
    rotationSpace=BCS_BoneSpace
    translationMode=BMM_Additive
    translationSpace=BCS_BoneSpace
    alphaBoolBlend {blendInTime=0.15, blendOption=Linear, blendOutTime=0.15}
    alphaScaleBiasClamp {bInterpResult=true, interpSpeedDecreasing=1, interpSpeedIncreasing=4}
    boneToModify {boneName=dyn_robe_r_front_01}
    rotation=<rot 0 20.0001 0>
    translation {z=5}
   }
   n15 = RigidBody!(ComponentPose=@n14, Alpha=Get JumpRun_Alpha() -> (JumpRun_Alpha:double), OverrideWorldGravity:Vector=Get GravityOverride(), ExternalForce) -> (Pose) {
    bClampLinearTranslationLimitToRefPose=true
    bOverrideWorldGravity=true
    overridePhysicsAsset=&/M_MED_Goat_Robe_CoatTails_Physics
    simulationSpace=BaseBoneSpace
    alphaBoolBlend {blendOption=Linear}
    baseBoneRef {boneName=pelvis}
    componentAppliedLinearAccClamp=<3000 3000 3000>
    componentLinearAccScale=<0.5 0.4 0.4>
    componentLinearVelScale=<0.5 0.4 0.4>
    overrideWorldGravity {z=-800}
    simSpaceSettings {
     maxAngularAcceleration=500
     maxAngularVelocity=3
     maxLinearAcceleration=400
     maxLinearVelocity=250
     velocityScaleZ=0
     worldAlpha=1
     externalLinearDragV=<0.1 0.1>
    }
   }
   n16 = RigidBody!(ComponentPose=@n15, Alpha, ExternalForce) -> (Pose) {
    bClampLinearTranslationLimitToRefPose=true
    bOverrideWorldGravity=true
    overridePhysicsAsset=&/M_MED_Goat_Robe_SideOrnaments_Physics
    simulationSpace=BaseBoneSpace
    alphaBoolBlend {blendOption=Linear}
    baseBoneRef {boneName=pelvis}
    componentAppliedLinearAccClamp=<6000 6000 6000>
    componentLinearAccScale=<0.25 0.25 0.25>
    componentLinearVelScale=<0.25 0.25 0.25>
    overrideWorldGravity {z=-980}
    simSpaceSettings {
     maxAngularAcceleration=500
     maxAngularVelocity=3
     maxLinearAcceleration=400
     maxLinearVelocity=250
     velocityScaleZ=0
     worldAlpha=1
     externalLinearDragV=<0.1 0.1>
    }
   }
   n17 = "Constraint - dyn_robe_r_front_01 (P)"(ComponentPose=@n16, Alpha=0.8, ConstraintWeights_0=1) {
    alphaBoolBlend {blendOption=Linear}
    alphaScaleBiasClamp {bClampResult=true, bInterpResult=true, clampMax=0.5, interpSpeedDecreasing=5, interpSpeedIncreasing=7}
    boneToModify {boneName=dyn_robe_r_front_01}
    constraintSetup [{targetBone {boneName=thigh_r}}]
    constraintWeights [1]
   }
   n18 = "Constraint - dyn_robe_l_front_01 (P)"(ComponentPose=@n17, Alpha=0.8, ConstraintWeights_0=1) {
    alphaBoolBlend {blendOption=Linear}
    alphaScaleBiasClamp {bClampResult=true, bInterpResult=true, clampMax=0.5, interpSpeedDecreasing=5, interpSpeedIncreasing=7}
    boneToModify {boneName=dyn_robe_l_front_01}
    constraintSetup [{targetBone {boneName=thigh_l}}]
    constraintWeights [1]
   }
   n19 = "Constraint - dyn_robe_r_front_02 (P)"(ComponentPose=@n18, Alpha=0.8, ConstraintWeights_0=1) {
    alphaBoolBlend {blendOption=Linear}
    alphaScaleBiasClamp {bClampResult=true, bInterpResult=true, clampMax=0.7, interpSpeedDecreasing=5, interpSpeedIncreasing=7}
    boneToModify {boneName=dyn_robe_r_front_02}
    constraintSetup [{targetBone {boneName=thigh_r}}]
    constraintWeights [1]
   }
   n20 = "Constraint - dyn_robe_l_front_02 (P)"(ComponentPose=@n19, Alpha=0.8, ConstraintWeights_0=1) {
    alphaBoolBlend {blendOption=Linear}
    alphaScaleBiasClamp {bClampResult=true, bInterpResult=true, clampMax=0.7, interpSpeedDecreasing=5, interpSpeedIncreasing=7}
    boneToModify {boneName=dyn_robe_l_front_02}
    constraintSetup [{targetBone {boneName=thigh_l}}]
    constraintWeights [1]
   }
   n21 = "Transform (Modify) Bone - Bone: dyn_robe_r_front_base"!(ComponentPose=@n20, Alpha=1, Translation="-0.000,5.5,0", Rotation="0,0,0") -> (Pose) {
    rotationMode=BMM_Replace
    rotationSpace=BCS_BoneSpace
    translationMode=BMM_Additive
    translationSpace=BCS_BoneSpace
    alphaBoolBlend {blendInTime=0.15, blendOption=Linear, blendOutTime=0.15}
    alphaScaleBiasClamp {bInterpResult=true, interpSpeedDecreasing=1, interpSpeedIncreasing=4}
    boneToModify {boneName=dyn_robe_r_front_base}
    translation {y=5.5}
   }
   n22 = Blend Poses by bool(
    BlendPose_0=Component To Local(ComponentPose=@n9)
    BlendPose_1=Component To Local(ComponentPose=@n21)
    BlendTime_1=0.25
    bActiveValue=Get Crouch_move() -> (Crouch_move:bool)
   ) {blendType=Linear}
   n23 = RigidBody(ComponentPose=Local To Component(LocalPose=Use cached pose 'GoatRobe_Cache'())) {
    bClampLinearTranslationLimitToRefPose=true
    bOverrideWorldGravity=true
    overridePhysicsAsset=&/M_MED_Goat_Robe_CoatTails_Physics_Emote
    simulationSpace=BaseBoneSpace
    alphaBoolBlend {blendOption=Linear}
    baseBoneRef {boneName=pelvis}
    componentAppliedLinearAccClamp=<3000 3000 3000>
    componentLinearAccScale=<0.1 0.1 0.1>
    componentLinearVelScale=<0.1 0.1 0.1>
    overrideWorldGravity {z=-800}
    simSpaceSettings {
     maxAngularAcceleration=500
     maxAngularVelocity=3
     maxLinearAcceleration=400
     maxLinearVelocity=250
     velocityScaleZ=0
     worldAlpha=1
     externalLinearDragV=<0.1 0.1>
    }
   }
   n24 = "Transform (Modify) Bone - Bone: dyn_robe_l_front_base"(ComponentPose=@n23, Alpha=1, Translation="0,0,0", Rotation="5,2,0", Scale="1,1,1") {
    rotationMode=BMM_Additive
    rotationSpace=BCS_BoneSpace
    translationMode=BMM_Additive
    translationSpace=BCS_BoneSpace
    alphaBoolBlend {blendOption=Linear}
    boneToModify {boneName=dyn_robe_l_front_base}
    rotation {pitch=5, yaw=2}
   }
   n25 = "Transform (Modify) Bone - Bone: dyn_robe_l_back_base"(ComponentPose=@n24, Alpha=1, Translation="0,0,1", Rotation="5,-5,0", Scale="1,1,1") {
    rotationMode=BMM_Additive
    rotationSpace=BCS_BoneSpace
    translationMode=BMM_Additive
    translationSpace=BCS_BoneSpace
    alphaBoolBlend {blendOption=Linear}
    boneToModify {boneName=dyn_robe_l_back_base}
    rotation {pitch=5, yaw=-5}
    translation {z=1}
   }
   n26 = "Transform (Modify) Bone - Bone: dyn_robe_l_mid_base"(ComponentPose=@n25, Alpha=1, Translation="0,0,1", Rotation="5,0,0", Scale="1,1,1") {
    rotationMode=BMM_Additive
    rotationSpace=BCS_BoneSpace
    translationMode=BMM_Additive
    translationSpace=BCS_BoneSpace
    alphaBoolBlend {blendOption=Linear}
    boneToModify {boneName=dyn_robe_l_mid_base}
    rotation {pitch=5}
    translation {z=1}
   }
   n27 = "Transform (Modify) Bone - Bone: dyn_robe_r_front_base"!(ComponentPose=@n26, Alpha=1, Translation="0,0,0", Rotation="-5,2,0", Scale:Vector="1,1,1") -> (Pose) {
    rotationMode=BMM_Additive
    rotationSpace=BCS_BoneSpace
    translationMode=BMM_Additive
    translationSpace=BCS_BoneSpace
    alphaBoolBlend {blendOption=Linear}
    boneToModify {boneName=dyn_robe_r_front_base}
    rotation {pitch=-5, yaw=2}
   }
   n28 = "Transform (Modify) Bone - Bone: dyn_robe_r_back_base"(ComponentPose=@n27, Alpha=1, Translation="0,0,1", Rotation="-5,-5,0", Scale="1,1,1") {
    rotationMode=BMM_Additive
    rotationSpace=BCS_BoneSpace
    translationMode=BMM_Additive
    translationSpace=BCS_BoneSpace
    alphaBoolBlend {blendOption=Linear}
    boneToModify {boneName=dyn_robe_r_back_base}
    rotation {pitch=-5, yaw=-5}
    translation {z=1}
   }
   n29 = "Transform (Modify) Bone - Bone: dyn_robe_r_mid_base"(ComponentPose=@n28, Alpha=1, Translation="0,0,1", Rotation="-5,0,0", Scale="1,1,1") {
    rotationMode=BMM_Additive
    rotationSpace=BCS_BoneSpace
    translationMode=BMM_Additive
    translationSpace=BCS_BoneSpace
    alphaBoolBlend {blendOption=Linear}
    boneToModify {boneName=dyn_robe_r_mid_base}
    rotation {pitch=-5}
    translation {z=1}
   }
   n30 = RigidBody(ComponentPose=@n29) {
    bClampLinearTranslationLimitToRefPose=true
    bOverrideWorldGravity=true
    overridePhysicsAsset=&M_MED_Goat_Robe_SideOrnaments_Physics
    simulationSpace=BaseBoneSpace
    alphaBoolBlend {blendOption=Linear}
    baseBoneRef {boneName=pelvis}
    componentAppliedLinearAccClamp=<6000 6000 6000>
    componentLinearAccScale=<0.25 0.25 0.25>
    componentLinearVelScale=<0.25 0.25 0.25>
    overrideWorldGravity {z=-980}
    simSpaceSettings {
     maxAngularAcceleration=500
     maxAngularVelocity=3
     maxLinearAcceleration=400
     maxLinearVelocity=250
     velocityScaleZ=0
     worldAlpha=1
     externalLinearDragV=<0.1 0.1>
    }
   }
   n31 = "Constraint - dyn_rope_01 (PPPPPPPPPPPP)"!(
    ComponentPose=@n30
    Alpha=0.3
    ConstraintWeights_0=0
    ConstraintWeights_1:float=0.0259
    ConstraintWeights_2:float=0.2866
    ConstraintWeights_3:float=0.0024
    ConstraintWeights_4:float=0.6849
    ConstraintWeights_5:float=0
    ConstraintWeights_6:float=0.0002
    ConstraintWeights_7:float=0
    ConstraintWeights_8:float=0
    ConstraintWeights_9:float=0
    ConstraintWeights_10:float=0
    ConstraintWeights_11:float=0
   ) -> (Pose) {
    alphaBoolBlend {blendOption=Linear}
    boneToModify {boneName=dyn_rope_01}
    constraintSetup [
     {targetBone {boneName=spine_02}}
     {targetBone {boneName=spine_01}}
     {targetBone {boneName=pelvis}}
     {targetBone {boneName=thigh_twist_01_r}}
     {targetBone {boneName=thigh_twist_01_l}}
     {targetBone {boneName=thigh_r}}
     {targetBone {boneName=thigh_l}}
     {targetBone {boneName=dyn_cloth_front_03}}
     {targetBone {boneName=dyn_cloth_front_01}}
     {targetBone {boneName=dyn_cloth_front_02}}
     {targetBone {boneName=dyn_robe_l_front_parent}}
     {targetBone {boneName=dyn_robe_l_front_01}}
    ]
    constraintWeights [1,1,1,1,1,1,1,1,1,1,1,1]
   }
   n32 = "Constraint - dyn_rope_02 (PPP)"!(ComponentPose=@n31, Alpha=0.5, ConstraintWeights_0=0.0082, ConstraintWeights_1:float=0.9851, ConstraintWeights_2:float=0.0067) -> (Pose) {
    alphaBoolBlend {blendOption=Linear}
    boneToModify {boneName=dyn_rope_02}
    constraintSetup [{targetBone {boneName=pelvis}}, {targetBone {boneName=thigh_twist_01_l}}, {targetBone {boneName=thigh_l}}]
    constraintWeights [1,1,1]
   }
   n33 = Component To Local(ComponentPose=@n32)
   n34 = Get bIsPlayingEmote()
   n35 = "Constraint - dyn_string_r_back_base (PPPPPP)"!(
    ComponentPose=Local To Component(LocalPose=Blend Poses by bool(BlendPose_0=@n33, BlendPose_1=@n22, BlendTime_1=0.25, bActiveValue=@n34) {blendType=Linear})
    Alpha=1
    ConstraintWeights_0=0.1181
    ConstraintWeights_1:float=0.1798
    ConstraintWeights_2:float=0.0003
    ConstraintWeights_3:float=0.7
    ConstraintWeights_4:float=0.0001
    ConstraintWeights_5:float=0.0017
   ) -> (Pose) {
    alphaBoolBlend {blendOption=Linear}
    boneToModify {boneName=dyn_string_r_back_base}
    constraintSetup [
     {targetBone {boneName=dyn_robe_r_back_03}}
     {targetBone {boneName=dyn_robe_r_mid_01}}
     {targetBone {boneName=dyn_robe_r_mid_02}}
     {targetBone {boneName=dyn_robe_r_mid_03}}
     {targetBone {boneName=dyn_robe_r_back_parent}}
     {targetBone {boneName=dyn_robe_r_back_01}}
    ]
    constraintWeights [1,1,1,1,1,1]
   }
   n36 = "Constraint - dyn_string_l_front_base (PPPP)"!(
    ComponentPose=@n35
    Alpha=1
    ConstraintWeights_0=0.0007
    ConstraintWeights_1:float=0.7
    ConstraintWeights_2:float=0.0062
    ConstraintWeights_3:float=0.2931
   ) -> (Pose) {
    alphaBoolBlend {blendOption=Linear}
    boneToModify {boneName=dyn_string_l_front_base}
    constraintSetup [
     {targetBone {boneName=dyn_robe_l_front_03}}
     {targetBone {boneName=dyn_robe_l_mid_01}}
     {targetBone {boneName=dyn_robe_l_front_01}}
     {targetBone {boneName=dyn_robe_l_front_02}}
    ]
    constraintWeights [1,1,1,1]
   }
   n37 = "Constraint - dyn_string_r_front_base (PPPP)"!(
    ComponentPose=@n36
    Alpha=1
    ConstraintWeights_0=0.18
    ConstraintWeights_1:float=0.7
    ConstraintWeights_2:float=0.1071
    ConstraintWeights_3:float=0.0129
   ) -> (Pose) {
    alphaBoolBlend {blendOption=Linear}
    boneToModify {boneName=dyn_string_r_front_base}
    constraintSetup [
     {targetBone {boneName=dyn_robe_r_front_03}}
     {targetBone {boneName=dyn_robe_r_mid_03}}
     {targetBone {boneName=dyn_robe_r_front_03}}
     {targetBone {boneName=dyn_robe_r_front_01}}
    ]
    constraintWeights [1,1,1,1]
   }
   n38 = "Constraint - dyn_string_l_back_base (PPPPPP)"!(
    ComponentPose=@n37
    Alpha=1
    ConstraintWeights_0=0.0049
    ConstraintWeights_1:float=0.5071
    ConstraintWeights_2:float=0.254
    ConstraintWeights_3:float=0.0018
    ConstraintWeights_4:float=0.2169
    ConstraintWeights_5:float=0.0153
   ) -> (Pose) {
    alphaBoolBlend {blendOption=Linear}
    boneToModify {boneName=dyn_string_l_back_base}
    constraintSetup [
     {targetBone {boneName=dyn_robe_l_back_03}}
     {targetBone {boneName=dyn_robe_l_mid_03}}
     {targetBone {boneName=dyn_robe_l_mid_01}}
     {targetBone {boneName=dyn_robe_l_back_01}}
     {targetBone {boneName=dyn_robe_l_back_02}}
     {targetBone {boneName=dyn_robe_l_back_03}}
    ]
    constraintWeights [1,1,1,1,1,1]
   }
   n39 = Anim Dynamics - Bone: dyn_string_l_back(ComponentPose=@n38, Alpha=0.5, ExternalForce="0,0,0") {
    angularDampingOverride=1
    bGravityOverrideInSimSpace=true
    bOverrideAngularDamping=true
    bOverrideLinearDamping=true
    linearDampingOverride=1
    alphaBoolBlend {blendOption=Linear}
    boundBone {boneName=dyn_string_l_back}
    componentAppliedLinearAccClamp=<6000 6000 6000>
    componentLinearAccScale=<0.25 0.25 0.25>
    componentLinearVelScale=<0.25 0.25 0.25>
    physicsBodyDefinitions [{
      sphereCollisionRadius=0
      boundBone {boneName=dyn_string_l_back}
      boxExtents=<10 3 3>
      constraintSetup {angularLimitsMax=<0 1 30>, angularLimitsMin=<0 -30 -30>, angularTarget=<0 0 0>}
      localJointOffset=<3 0 0>
     }]
    relativeSpaceBone {boneName=spine_01}
   }
   n40 = Anim Dynamics - Bone: dyn_string_l_front(ComponentPose=@n39, Alpha=0.5, ExternalForce="0,0,0") {
    angularDampingOverride=1
    bGravityOverrideInSimSpace=true
    bOverrideAngularDamping=true
    bOverrideLinearDamping=true
    linearDampingOverride=1
    alphaBoolBlend {blendOption=Linear}
    boundBone {boneName=dyn_string_l_front}
    componentAppliedLinearAccClamp=<6000 6000 6000>
    componentLinearAccScale=<0.25 0.25 0.25>
    componentLinearVelScale=<0.25 0.25 0.25>
    physicsBodyDefinitions [{
      sphereCollisionRadius=0
      boundBone {boneName=dyn_string_l_front}
      boxExtents=<10 3 3>
      constraintSetup {angularLimitsMax=<0 1 30>, angularLimitsMin=<0 -30 -30>, angularTarget=<0 0 0>}
      localJointOffset=<3 0 0>
     }]
    relativeSpaceBone {boneName=spine_01}
   }
   n41 = Anim Dynamics - Bone: dyn_string_r_back(ComponentPose=@n40, Alpha=0.5, ExternalForce="0,0,0") {
    angularDampingOverride=1
    bGravityOverrideInSimSpace=true
    bOverrideAngularDamping=true
    bOverrideLinearDamping=true
    linearDampingOverride=1
    alphaBoolBlend {blendOption=Linear}
    boundBone {boneName=dyn_string_r_back}
    componentAppliedLinearAccClamp=<6000 6000 6000>
    componentLinearAccScale=<0.25 0.25 0.25>
    componentLinearVelScale=<0.25 0.25 0.25>
    physicsBodyDefinitions [{
      sphereCollisionRadius=0
      boundBone {boneName=dyn_string_r_back}
      boxExtents=<10 3 3>
      constraintSetup {angularLimitsMax=<0 1 30>, angularLimitsMin=<0 -30 -30>, angularTarget=<0 0 0>}
      localJointOffset=<3 0 0>
     }]
    relativeSpaceBone {boneName=spine_01}
   }
   n42 = Anim Dynamics - Bone: dyn_string_r_front(ComponentPose=@n41, Alpha=0.5, ExternalForce="0,0,0") {
    angularDampingOverride=1
    bGravityOverrideInSimSpace=true
    bOverrideAngularDamping=true
    bOverrideLinearDamping=true
    linearDampingOverride=1
    alphaBoolBlend {blendOption=Linear}
    boundBone {boneName=dyn_string_r_front}
    componentAppliedLinearAccClamp=<6000 6000 6000>
    componentLinearAccScale=<0.25 0.25 0.25>
    componentLinearVelScale=<0.25 0.25 0.25>
    physicsBodyDefinitions [{
      sphereCollisionRadius=0
      boundBone {boneName=dyn_string_r_front}
      boxExtents=<10 3 3>
      constraintSetup {angularLimitsMax=<0 1 30>, angularLimitsMin=<0 -30 -30>, angularTarget=<0 0 0>}
      localJointOffset=<3 0 0>
     }]
    relativeSpaceBone {boneName=spine_01}
   }
   n43 = Anim Dynamics - Bone: dyn_fur_01(ComponentPose=@n42, Alpha=0.5, ExternalForce="0,0,0") {
    angularDampingOverride=0.999
    angularSpringConstant=0.99
    bAngularSpring=true
    bGravityOverrideInSimSpace=true
    bOverrideAngularDamping=true
    bOverrideLinearDamping=true
    linearDampingOverride=0.999
    simulationSpace=BoneRelative
    alphaBoolBlend {blendOption=Linear}
    boundBone {boneName=dyn_fur_01}
    componentAppliedLinearAccClamp=<6000 6000 6000>
    componentLinearAccScale=<0.25 0.25 0.25>
    componentLinearVelScale=<0.25 0.25 0.25>
    physicsBodyDefinitions [{
      sphereCollisionRadius=0
      boundBone {boneName=dyn_fur_01}
      boxExtents=<10 5 5>
      constraintSetup {angularLimitsMax=<0 30 30>, angularLimitsMin=<0 -30 -30>, angularTarget=<1 0 0>}
      localJointOffset=<2 0 0>
     }]
    relativeSpaceBone {boneName=spine_05}
   }
   n44 = Anim Dynamics - Bone: dyn_fur_02(ComponentPose=@n43, Alpha=0.5, ExternalForce="0,0,0") {
    angularDampingOverride=0.999
    angularSpringConstant=0.99
    bAngularSpring=true
    bGravityOverrideInSimSpace=true
    bOverrideAngularDamping=true
    bOverrideLinearDamping=true
    linearDampingOverride=0.999
    simulationSpace=BoneRelative
    alphaBoolBlend {blendOption=Linear}
    boundBone {boneName=dyn_fur_02}
    componentAppliedLinearAccClamp=<6000 6000 6000>
    componentLinearAccScale=<0.25 0.25 0.25>
    componentLinearVelScale=<0.25 0.25 0.25>
    physicsBodyDefinitions [{
      sphereCollisionRadius=0
      boundBone {boneName=dyn_fur_02}
      boxExtents=<10 5 5>
      constraintSetup {angularLimitsMax=<0 35 35>, angularLimitsMin=<0 -35 -35>, angularTarget=<1 0 0>}
      localJointOffset=<2 0 0>
     }]
    relativeSpaceBone {boneName=spine_05}
   }
   n45 = Anim Dynamics - Bone: dyn_fur_03(ComponentPose=@n44, Alpha=0.5, ExternalForce="0,0,0") {
    angularDampingOverride=0.999
    angularSpringConstant=0.99
    bAngularSpring=true
    bGravityOverrideInSimSpace=true
    bOverrideAngularDamping=true
    bOverrideLinearDamping=true
    linearDampingOverride=0.999
    simulationSpace=BoneRelative
    alphaBoolBlend {blendOption=Linear}
    boundBone {boneName=dyn_fur_03}
    componentAppliedLinearAccClamp=<6000 6000 6000>
    componentLinearAccScale=<0.25 0.25 0.25>
    componentLinearVelScale=<0.25 0.25 0.25>
    physicsBodyDefinitions [{
      sphereCollisionRadius=0
      boundBone {boneName=dyn_fur_03}
      boxExtents=<10 5 5>
      constraintSetup {angularLimitsMax=<0 30 30>, angularLimitsMin=<0 -30 -30>, angularTarget=<1 0 0>}
      localJointOffset=<2 0 0>
     }]
    relativeSpaceBone {boneName=spine_05}
   }
   n46 = Anim Dynamics - Bone: dyn_fur_04(ComponentPose=@n45, Alpha=0.5, ExternalForce="0,0,0") {
    angularDampingOverride=0.999
    angularSpringConstant=0.99
    bAngularSpring=true
    bGravityOverrideInSimSpace=true
    bOverrideAngularDamping=true
    bOverrideLinearDamping=true
    linearDampingOverride=0.999
    simulationSpace=BoneRelative
    alphaBoolBlend {blendOption=Linear}
    boundBone {boneName=dyn_fur_04}
    componentAppliedLinearAccClamp=<6000 6000 6000>
    componentLinearAccScale=<0.25 0.25 0.25>
    componentLinearVelScale=<0.25 0.25 0.25>
    physicsBodyDefinitions [{
      sphereCollisionRadius=0
      boundBone {boneName=dyn_fur_04}
      boxExtents=<10 5 5>
      constraintSetup {angularLimitsMax=<0 35 35>, angularLimitsMin=<0 -35 -30>, angularTarget=<1 0 0>}
      localJointOffset=<2 0 0>
     }]
    relativeSpaceBone {boneName=spine_05}
   }
   n47 = Anim Dynamics - Bone: dyn_fur_05(ComponentPose=@n46, Alpha=0.5, ExternalForce="0,0,0") {
    angularDampingOverride=0.999
    angularSpringConstant=0.99
    bAngularSpring=true
    bGravityOverrideInSimSpace=true
    bOverrideAngularDamping=true
    bOverrideLinearDamping=true
    linearDampingOverride=0.999
    simulationSpace=BoneRelative
    alphaBoolBlend {blendOption=Linear}
    boundBone {boneName=dyn_fur_05}
    componentAppliedLinearAccClamp=<6000 6000 6000>
    componentLinearAccScale=<0.25 0.25 0.25>
    componentLinearVelScale=<0.25 0.25 0.25>
    physicsBodyDefinitions [{
      sphereCollisionRadius=0
      boundBone {boneName=dyn_fur_05}
      boxExtents=<10 5 5>
      constraintSetup {angularLimitsMax=<0 30 30>, angularLimitsMin=<0 -30 -30>, angularTarget=<1 0 0>}
      localJointOffset=<2 0 0>
     }]
    relativeSpaceBone {boneName=spine_05}
   }
   n48 = Anim Dynamics - Bone: dyn_fur_06(ComponentPose=@n47, Alpha=0.5, ExternalForce="0,0,0") {
    angularDampingOverride=0.999
    angularSpringConstant=0.99
    bAngularSpring=true
    bGravityOverrideInSimSpace=true
    bOverrideAngularDamping=true
    bOverrideLinearDamping=true
    linearDampingOverride=0.999
    simulationSpace=BoneRelative
    alphaBoolBlend {blendOption=Linear}
    boundBone {boneName=dyn_fur_06}
    componentAppliedLinearAccClamp=<6000 6000 6000>
    componentLinearAccScale=<0.25 0.25 0.25>
    componentLinearVelScale=<0.25 0.25 0.25>
    physicsBodyDefinitions [{
      sphereCollisionRadius=0
      boundBone {boneName=dyn_fur_06}
      boxExtents=<10 5 5>
      constraintSetup {angularLimitsMax=<0 35 35>, angularLimitsMin=<0 -35 -35>, angularTarget=<1 0 0>}
      localJointOffset=<2 0 0>
     }]
    relativeSpaceBone {boneName=spine_05}
   }
   n49 = Anim Dynamics - Bone: dyn_book_parent(ComponentPose=@n48, Alpha=1, ExternalForce="0,0,0") {
    angularDampingOverride=0.9999
    angularSpringConstant=500
    bAngularSpring=true
    bGravityOverrideInSimSpace=true
    bOverrideAngularDamping=true
    simulationSpace=BoneRelative
    alphaBoolBlend {blendOption=Linear}
    boundBone {boneName=dyn_book_parent}
    componentAppliedLinearAccClamp=<3500 3500 3500>
    componentLinearAccScale=<1.2 1.2 1.2>
    componentLinearVelScale=<1.2 1.2 1.2>
    gravityOverride {z=-980}
    physicsBodyDefinitions [{
      sphereCollisionRadius=0
      boundBone {boneName=dyn_book_parent}
      boxExtents=<10 20 20>
      constraintSetup {angularLimitsMax=<1 5 8>, angularLimitsMin=<-1 -90 -8>, angularTarget=<0 0 0>}
      localJointOffset=<5 0 0>
     }]
    relativeSpaceBone {boneName=root}
   }
   n50 = "Transform (Modify) Bone - Bone: dyn_fur_07"(ComponentPose=@n49, Alpha=@n3, Translation="0,0,2", Rotation="0,50,0", Scale="1,1,1") {
    rotationMode=BMM_Additive
    rotationSpace=BCS_BoneSpace
    translationMode=BMM_Additive
    translationSpace=BCS_BoneSpace
    alphaBoolBlend {blendOption=Linear}
    alphaScaleBiasClamp {bInterpResult=true, interpSpeedDecreasing=7, interpSpeedIncreasing=7}
    boneToModify {boneName=dyn_fur_07}
    rotation {yaw=50}
    translation {z=2}
   }
   n51 = "Transform (Modify) Bone - Bone: dyn_fur_06"(ComponentPose=@n50, Alpha=@n3, Translation="0,10,10", Rotation="0,60,0", Scale="1,1,1") {
    rotationMode=BMM_Additive
    rotationSpace=BCS_BoneSpace
    translationMode=BMM_Additive
    translationSpace=BCS_BoneSpace
    alphaBoolBlend {blendOption=Linear}
    alphaScaleBiasClamp {bInterpResult=true, interpSpeedDecreasing=7, interpSpeedIncreasing=7}
    boneToModify {boneName=dyn_fur_06}
    rotation {yaw=60}
    translation {y=10, z=10}
   }
   n52 = "Transform (Modify) Bone - Bone: dyn_fur_05"(ComponentPose=@n51, Alpha=@n3, Translation="0,10,10", Rotation="0,40,0", Scale="1,1,1") {
    rotationMode=BMM_Additive
    rotationSpace=BCS_BoneSpace
    translationMode=BMM_Additive
    translationSpace=BCS_BoneSpace
    alphaBoolBlend {blendOption=Linear}
    alphaScaleBiasClamp {bInterpResult=true, interpSpeedDecreasing=7, interpSpeedIncreasing=7}
    boneToModify {boneName=dyn_fur_05}
    rotation {yaw=40}
    translation {y=10, z=10}
   }
   n53 = "Transform (Modify) Bone - Bone: dyn_fur_parent"(ComponentPose=@n52, Alpha=@n3, Translation="0,0,0", Rotation="0,20,0", Scale="1,1,1") {
    rotationMode=BMM_Additive
    rotationSpace=BCS_BoneSpace
    alphaBoolBlend {blendOption=Linear}
    boneToModify {boneName=dyn_fur_parent}
    rotation {yaw=20}
   }
   n54 = Anim Dynamics - Bone: dyn_scarf_front_parent(ComponentPose=@n53, Alpha=1, ExternalForce="0,0,0") {
    angularDampingOverride=0.9999
    angularSpringConstant=500
    bAngularSpring=true
    bGravityOverrideInSimSpace=true
    bOverrideAngularDamping=true
    simulationSpace=BoneRelative
    alphaBoolBlend {blendOption=Linear}
    boundBone {boneName=dyn_scarf_front_parent}
    componentAppliedLinearAccClamp=<5000 5000 5000>
    componentLinearAccScale=<1.2 1.2 1.2>
    componentLinearVelScale=<1.2 1.2 1.2>
    physicsBodyDefinitions [{
      sphereCollisionRadius=0
      boundBone {boneName=dyn_scarf_front_parent}
      boxExtents=<15 20 20>
      constraintSetup {angularLimitsMax=<1 5 25>, angularLimitsMin=<-1 -45 -5>, angularTarget=<0 0 0>}
      localJointOffset=<3 0 0>
     }]
    relativeSpaceBone {boneName=root}
   }
   n55 = Anim Dynamics - Bone: dyn_scarf_front_01(ComponentPose=@n54, Alpha=1, ExternalForce="0,0,0") {
    angularDampingOverride=0.9999
    angularSpringConstant=500
    bAngularSpring=true
    bGravityOverrideInSimSpace=true
    bOverrideAngularDamping=true
    simulationSpace=BoneRelative
    alphaBoolBlend {blendOption=Linear}
    boundBone {boneName=dyn_scarf_front_01}
    componentAppliedLinearAccClamp=<5000 5000 5000>
    componentLinearAccScale=<1.2 1.2 1.2>
    componentLinearVelScale=<1.2 1.2 1.2>
    physicsBodyDefinitions [{
      sphereCollisionRadius=0
      boundBone {boneName=dyn_scarf_front_01}
      boxExtents=<15 20 20>
      constraintSetup {angularLimitsMax=<1 8 40>, angularLimitsMin=<-1 -45 -5>, angularTarget=<0 0 0>}
      localJointOffset=<3 0 0>
     }]
    relativeSpaceBone {boneName=root}
   }
   n56 = Anim Dynamics - Bone: dyn_scarf_front_02(ComponentPose=@n55, Alpha=1, ExternalForce="0,0,0") {
    angularDampingOverride=0.9999
    angularSpringConstant=500
    bAngularSpring=true
    bGravityOverrideInSimSpace=true
    bOverrideAngularDamping=true
    simulationSpace=BoneRelative
    alphaBoolBlend {blendOption=Linear}
    boundBone {boneName=dyn_scarf_front_02}
    componentAppliedLinearAccClamp=<5000 5000 5000>
    componentLinearAccScale=<1.2 1.2 1.2>
    componentLinearVelScale=<1.2 1.2 1.2>
    physicsBodyDefinitions [{
      sphereCollisionRadius=0
      boundBone {boneName=dyn_scarf_front_02}
      boxExtents=<15 20 20>
      constraintSetup {angularLimitsMax=<1 8 25>, angularLimitsMin=<-1 -45 -5>, angularTarget=<0 0 0>}
      localJointOffset=<3 0 0>
     }]
    relativeSpaceBone {boneName=root}
   }
   n57 = Look At - Bone: dyn_robe_short_r_back!(ComponentPose=@n56, bAlphaBoolEnabled:bool=Get bIsCrouching()) -> (Pose) {
    alphaInputType=Bool
    alphaBoolBlend {blendOption=Linear}
    boneToModify {boneName=dyn_robe_short_r_back}
    lookAtLocation=<0 -100>
    lookAtTarget {boneReference {boneName=calf_r}}
   }
   Output Pose(Result=Component To Local(ComponentPose=@n57)) {layerGroup=None}
   n58 = "Transform (Modify) Bone - Bone: dyn_rope_base"(
    ComponentPose=Local To Component(LocalPose=Copy Pose From Mesh() {bUseAttachedParent=true})
    Alpha=Get Crouch_Alpha() -> (Crouch_Alpha:double)
    Translation="0,-10,0"
    Rotation="0,-40,-60"
    Scale="1,1,1"
   ) {
    rotationMode=BMM_Additive
    rotationSpace=BCS_BoneSpace
    translationMode=BMM_Additive
    translationSpace=BCS_BoneSpace
    alphaBoolBlend {blendOption=Linear}
    alphaScaleBiasClamp {bInterpResult=true, interpSpeedDecreasing=7}
    boneToModify {boneName=dyn_rope_base}
    rotation {roll=-60, yaw=-40}
    translation {y=-10}
   }
   n59 = Anim Dynamics - Bone: dyn_scarf_back_parent(ComponentPose=@n58, Alpha=1, ExternalForce="0,0,0") {
    angularDampingOverride=0.9999
    angularSpringConstant=500
    bAngularSpring=true
    bGravityOverrideInSimSpace=true
    bOverrideAngularDamping=true
    simulationSpace=BoneRelative
    alphaBoolBlend {blendOption=Linear}
    boundBone {boneName=dyn_scarf_back_parent}
    componentAppliedLinearAccClamp=<3500 3500 3500>
    componentLinearAccScale=<1.2 1.2 1.2>
    componentLinearVelScale=<1.2 1.2 1.2>
    physicsBodyDefinitions [{
      sphereCollisionRadius=0
      boundBone {boneName=dyn_scarf_back_parent}
      boxExtents=<15 20 20>
      constraintSetup {angularLimitsMax=<1 15 5>, angularLimitsMin=<-1 -15 -60>, angularTarget=<0 0 0>}
      localJointOffset=<4 0 0>
     }]
    relativeSpaceBone {boneName=root}
   }
   n60 = Anim Dynamics - Bone: dyn_scarf_back_01(ComponentPose=@n59, Alpha=1, ExternalForce="0,0,0") {
    angularDampingOverride=0.9999
    angularSpringConstant=500
    bAngularSpring=true
    bGravityOverrideInSimSpace=true
    bOverrideAngularDamping=true
    simulationSpace=BoneRelative
    alphaBoolBlend {blendOption=Linear}
    boundBone {boneName=dyn_scarf_back_01}
    componentAppliedLinearAccClamp=<3500 3500 3500>
    componentLinearAccScale=<1.2 1.2 1.2>
    componentLinearVelScale=<1.2 1.2 1.2>
    physicsBodyDefinitions [{
      sphereCollisionRadius=0
      boundBone {boneName=dyn_scarf_back_01}
      boxExtents=<15 20 20>
      constraintSetup {angularLimitsMax=<1 15 5>, angularLimitsMin=<-1 -15 -25>, angularTarget=<0 0 0>}
      localJointOffset=<4 0 0>
     }]
    relativeSpaceBone {boneName=root}
   }
   n61 = "Transform (Modify) Bone - Bone: dyn_scarf_back_base"(ComponentPose=@n60, Alpha=1, Translation="-1.1,-1.4,-1.1", Rotation="0,-5,5", Scale="1,1,1") {
    rotationMode=BMM_Additive
    rotationSpace=BCS_BoneSpace
    translationMode=BMM_Additive
    translationSpace=BCS_BoneSpace
    alphaBoolBlend {blendOption=Linear}
    boneToModify {boneName=dyn_scarf_back_base}
    rotation {roll=5, yaw=-5}
    translation=<-1.1 -1.4 -1.1>
   }
   Save cached pose 'GoatRobe_Cache'(Pose=Component To Local(ComponentPose=@n61)) {cacheName=GoatRobe_Cache}
   Reroute Node(InputPin:Wildcard) -> (OutputPin:Wildcard)
  ]}
 EventGraph {nodes [
   Event BlueprintUpdateAnimation()
   > Get!(execute:Exec) -> (then:Exec, else:Exec, player_pawn:Fort Player Pawn Athena Object Reference)
   > n1 = Sequence!(execute) -> (then_0, then_1, then_2:Exec, then_3:Exec, then_4:Exec, then_5:Exec, then_6:Exec)
   n2 = Gravity Override State Function DEPRECATED(
    execute=@n1.then_6
    CharacterPartAnimInstance=Self-Reference()
    GameTimeInSeconds=GetGameTimeInSeconds()
    GravityOverrideParams=Get Gravity Override Params()
   )
   > Set GravityOverride(GravityOverride=@n2)
   > n3 = Branch(Condition=IsCosmeticObjectDisplayedInFrontendContext!(CosmeticObject:Object Reference) -> (ReturnValue:bool))
   > Set frontend alpha(frontend alpha:double=0) -> (Output_Get:double)
   Set frontend alpha(execute=@n3.else, frontend alpha=0.8)
   n4 = Get player_pawn() -> (player_pawn:Fort Player Pawn Athena Object Reference)
   Is Valid(exec=@n1.then_0, InputObject=@n4)
   > Do Once()
   > Set isBackpackEquipped(isBackpackEquipped:bool="Not Equal (Object)"!(A:Object Reference=GetSkeletalMeshForPartType(self=@n4, PartType=Backpack), B:Object Reference) -> (ReturnValue:bool)) -> (Output_Get:bool)
   n5 = "float >= float"(A=Get PawnSpeed2D(), B=545) {is=title K2Node_CallFunction}
   n6 = Get bIsJumping()
   n7 = Get bIsFalling()
   Set JumpRun_Alpha(execute=@n1.then_2, JumpRun_Alpha:double=SelectFloat(A=0.5, B=1, bPickA=OR Boolean!(A=@n6, B=@n7, C:bool=@n5) -> (ReturnValue))) -> (Output_Get:double)
   n8 = GetOwnerSkeletalMeshComponent()
   n9 = BreakRotator(InRot=BreakTransform(InTransform=GetDeltaTransformFromRefPose(self=@n8, BoneName=upperarm_l, BaseName=clavicle_l)).Rotation)
   Set armSwingDelta_l(execute=@n1.then_4, armSwingDelta_l:double=MapRangeClamped(Value="Absolute (Float)"(A=@n9.Yaw), InRangeA=15, InRangeB=82)) -> (Output_Get:double)
   Set Jump_Alpha(execute=@n1.then_5, Jump_Alpha:double=SelectFloat(A=1, B=0, bPickA=Get bIsJumping())) -> (Output_Get:double)
   > Set Backpack On(Backpack On:double=SelectFloat(A=0, B=1, bPickA=Get isBackpackEquipped() -> (isBackpackEquipped:bool))) -> (Output_Get:double)
   > Set Crouch_move(Crouch_move:bool=OR Boolean!(A=Get bIsCrouching(), B=Get bIsCrouchSprinting()) -> (ReturnValue)) -> (Output_Get:bool)
   n10 = "integer > integer"(A=Get CurrentLOD(), B=1) {is=title K2Node_CallFunction}
   n11 = OR Boolean(A=NOT Boolean(A=Get bRigidBodyEnabled()), B=@n10)
   Set isRigidBodyDisabled(execute=@n1.then_1, isRigidBodyDisabled:bool=@n11) -> (Output_Get:bool)
   > Set isBackStatic(isBackStatic:bool=OR Boolean(A=Get isBackpackEquipped(), B=@n11)) -> (Output_Get:bool)
   > Set back_collar_lod_offset(back_collar_lod_offset:Vector=SelectVector(A="0,-5, 0", B="0, 0, 0", bPickA=@n10)) -> (Output_Get:Vector)
   > Set back_collar_lod_rot_offset(back_collar_lod_rot_offset:Rotator=SelectRotator!(A:Rotator="0,20, 0", B:Rotator="0, 0, 0", bPickA:bool=@n10) -> (ReturnValue:Rotator)) -> (Output_Get:Rotator)
   n12 = "To Float (Boolean)"(InBool=Get bIsCrouching())
   Set Crouch_Alpha(execute=@n1.then_3, Crouch_Alpha:double=@n12) -> (Output_Get:double)
   > Do Once(Reset=[@n15.then, @n16.then])
   > Delay(Duration=0.212)
   > Set CrouchRigid(CrouchRigid:double=@n12) -> (Output_Get:double)
   Event BlueprintInitializeAnimation()
   n13 = TryGetPawnOwner()
   > Is Valid(InputObject=@n13)
   > n14 = Cast To FortPlayerPawnAthena(Object=@n13)
   > Set player_pawn(player_pawn:Fort Player Pawn Athena Object Reference=@n14) -> (Output_Get:Fort Player Pawn Athena Object Reference)
   n15 = Event OnEndCrouching()
   n16 = Event OnBeginCrouching()
  ]}
]
variables [
 armSwingDelta_l {type=number}
 back_collar_lod_offset {type=Vector}
 back_collar_lod_rot_offset {type=Rotator}
 Backpack On {type=number}
 Crouch_Alpha {type=number}
 Crouch_move {type=boolean}
 CrouchRigid {type=number}
 frontend alpha {type=number}
 Gravity Override Params {type=GravityOverrideParamsStruct}
 GravityOverride {type=Vector}
 isBackpackEquipped {type=boolean}
 isBackStatic {type=boolean}
 isRigidBodyDisabled {type=boolean}
 Jump_Alpha {type=number}
 JumpRun_Alpha {type=number}
 player_pawn {type=FortPlayerPawnAthena}
]
